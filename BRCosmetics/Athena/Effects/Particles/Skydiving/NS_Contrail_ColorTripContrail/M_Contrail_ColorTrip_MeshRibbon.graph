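# Translucent, two-sided lit material for the ColorTrip skydiving contrail, used on Niagara mesh particles and sprites.
# Use_StickerSubUV blends between a particle-colored ribbon and a tinted sticker texture with near-camera fade.
# The ribbon's opacity combines a dynamic-parameter-panned gradient with a sine V mask.
# SplineThicken sets ribbon width while scrolling noise adds world position offset, both driven by dynamic parameters; sticker mode zeroes this offset.
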Material /BRCosmetics/Athena/Effects/Particles/Skydiving/NS_Contrail_ColorTripContrail/M_Contrail_ColorTrip_MeshRibbon
graph {
 properties {
  bUsedWithNiagaraMeshParticles=true
  bUsedWithNiagaraSprites=true
  blendMode=BLEND_Translucent
  pixelDepthOffsetMode=PDOM_Legacy
  twoSided=true
  usedShadingModels=Default Lit
 }
 nodes [
  n1 = TextureSampleParameter2D() {
   group=Simple_texture
   parameterName=Simple_Texture
   texture=&/T_2D_sticker
  }
  n2 = ParticleColor()
  n3 = Multiply(A=Desaturation(None=@n1.RGB, Fraction=ScalarParameter() {group=Simple_texture, parameterName=Des}), B=@n2.RGB)
  n4 = ParticleColor()
  LinearInterpolate(A=@n4.RGB, B=@n3, Alpha=ScalarParameter() {parameterName=Use_StickerSubUV})
  n5 = ScalarParameter() {defaultValue=256, parameterName=Near_Camera_Fade}
  n6 = Add(A=TextureCoordinate(), B=LinearInterpolate(Alpha=DynamicParameter() {paramNames [Panner,Width,WPO,Param4]}.Panner) {constA=-0.8, constB=0.8}) {constB=-0.8}
  n7 = Saturate(None=Multiply(A=TextureSample(UVs=@n6) {samplerSource=SSM_Clamp_WorldGroupSettings, texture=/Game/Global/Textures/Default/Masks_and_Gradients/gradient_stasis}.R) {constB=10})
  n8 = ComponentMask(None=TextureCoordinate()) {g=true}
  n9 = Saturate(None=Multiply(A=Sine(None=@n8) {period=2}) {constB=2})
  LinearInterpolate(
   A=Multiply(A=@n4.A, B=Multiply(A=@n7, B=@n9))
   B=Multiply(
    A=Multiply(A=@n1.A, B=@n2.A)
    B=MaterialFunctionCall(Distance=@n5) {materialFunction=&Near_Camera_Fade}.Result
   )
   Alpha=ScalarParameter() {parameterName=Use_StickerSubUV}
  )
  n10 = DynamicParameter() {paramNames [Panner,Width,WPO,Param4]}
  n11 = Multiply(A=@n10.Width) {constB=0.3}
  n12 = TextureCoordinate() {uTiling=0}
  n13 = Panner(Coordinate=TextureCoordinate() {uTiling=0.4, vTiling=0.35}) {speedX=0.1, speedY=0.2}
  n14 = Multiply(A=ComponentMask(None=TextureSample(UVs=@n13) {samplerType=SAMPLERTYPE_Normal, texture=&T_Weap_Distort_N_02}.RGB) {g=true, r=true}) {constB=0}
  n15 = TextureSample(UVs=Panner(Coordinate=Add(A=@n12, B=@n14)) {speedY=1}) {samplerType=SAMPLERTYPE_Grayscale, texture=&LowResBlurredNoise}
  n16 = Multiply(A=Constant3Vector() {constant=<rgba 1.5 1.5 0 0>}, B=Power(Base=ConstantBiasScale(None=@n15.R) {bias=0, scale=2}))
  n17 = Transform(None=Multiply(A=Multiply(A=@n16), B=DynamicParameter() {paramNames [Panner,Width,WPO,Param4]}.WPO)) {transformSourceType=TRANSFORMSOURCE_Local}
  LinearInterpolate(
   A=Add(
    A=MaterialFunctionCall(WidthBase=@n10.Width, WidthTip=@n11) {materialFunction=&SplineThicken}.WorldPositionOffset
    B=Multiply(A=@n17)
   )
   B=Constant()
   Alpha=ScalarParameter() {parameterName=Use_StickerSubUV}
  )
  n18 = ScalarParameter() {defaultValue=1, group=Simple_texture, parameterName=Brightness}
  MaterialFunctionCall(Emissive=LinearInterpolate(A=Multiply(A=@n4.RGB) {constB=0.3}, B=Multiply(A=@n3, B=@n18), Alpha=ScalarParameter() {parameterName=Use_StickerSubUV})) {materialFunction=&MF_FX_EmissiveAdjustments}
 ]
}
