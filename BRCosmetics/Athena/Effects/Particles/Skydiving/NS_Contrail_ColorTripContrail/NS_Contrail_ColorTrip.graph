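# Niagara skydiving contrail cosmetic "Color Trip": multicolored trails, stickers and sparkles streaming from the player's hands and feet.
# Two paint ribbon emitters spawn on hand bones, with upward velocity scaled by owner speed and colors picked from curves per spawn group.
# Sub-UV sticker sprites, GPU drops and sparkles, and randomly colored meshes follow skeletal bones.
# An attached glow cycles colors on sprites and lights, dimmed by time of day.
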NiagaraSystem /BRCosmetics/Athena/Effects/Particles/Skydiving/NS_Contrail_ColorTripContrail/NS_Contrail_ColorTrip
dependencies {
 usedDataInterfaces [&NiagaraDataInterfaceColorCurve, &NiagaraDataInterfaceCurve]
 usedDynamicInputs []
 usedModules [&SystemState]
}
emitters {
 AttachedGlow {
  emitterUpdateScript [
   &EmitterState(Inactive Response=::NewEnumerator1, Life Cycle Mode=::NewEnumerator1, Loop Duration=0, UseLoopDelay=false)
   &SpawnRate(SpawnRate=10, Use Spawn Probability=false)
  ]
  particleSpawnScript [
   &FN_SetSpawnGroup_v2(Number of Spawn Groups=4)
   &InitializeParticle(
    Color=<rgba 1 1 1 2>
    Lifetime=0.6
    Sprite Size Mode=::NewEnumerator1
    Uniform Sprite Size Max=150
    Uniform Sprite Size Min=100
    UsePositionOffset=false
    Write Material Random=true
   ) {module=InitializeParticle001}
   &CameraOffset()
   /SetVariables(Particles.LightRadius=45) {isSetParameters=true}
  ]
  particleUpdateScript [
   /SetVariables(Particles.Color=&MakeLinearColorFromVectorAndFloat(
     "Vector (RGB)"=&MakeVectorFromLinearColorRGB(Color=&ColorFromCurve(ColorCurve=DataInterface() {propertyValues {
         AlphaCurve {keys [<at 0 1 RCIM_Cubic>,<at 0.9977 1 RCIM_Cubic>]}
         BlueCurve {keys [<at 0 1 RCIM_Cubic>,<at 0.341 0.129>,<at 0.5867 1>,<at 0.7956 0>,<at 1 0.9502 RCIM_Cubic>]}
         GreenCurve {keys [<at 0 0.3902 RCIM_Cubic>,<at 0.341 0>,<at 0.5867 0.5665>,<at 0.7956 0.6287>,<at 1 1 RCIM_Cubic>]}
         RedCurve {keys [<at 0 0.027 RCIM_Cubic>,<at 0.341 1>,<at 0.5867 0.031>,<at 0.7956 1>,<at 1 0.0299 RCIM_Cubic>]}
        }}))
     "Float (Alpha)"=&FloatFromCurve(
      FloatCurve=DataInterface() {propertyValues {Curve {keys [<at 0 0 RCIM_Cubic>, {arriveTangent=0.002, interpMode=RCIM_Cubic, leaveTangent=0.002, time=0.5, value=0.1}, <at 1.0011 0.002>]}}}
      Scale Curve=1.2
     )
    )) {isSetParameters=true}
   /SetVariables(Particles.Light_color:LinearColor=&MultiplyLinearColor(
     Linear Color=&ColorFromCurve(ColorCurve=DataInterface() {propertyValues {
        AlphaCurve {keys [<at 0 1 RCIM_Cubic>,<at 0.9977 1 RCIM_Cubic>]}
        BlueCurve {keys [<at 0 1 RCIM_Cubic>,<at 0.341 0.129>,<at 0.5867 1>,<at 0.7956 0>,<at 1 0.9502 RCIM_Cubic>]}
        GreenCurve {keys [<at 0 0.3902 RCIM_Cubic>,<at 0.341 0>,<at 0.5867 0.5665>,<at 0.7956 0.6288>,<at 1 1 RCIM_Cubic>]}
        RedCurve {keys [<at 0 0.027 RCIM_Cubic>,<at 0.341 1>,<at 0.5867 0.031>,<at 0.7956 1>,<at 1 0.0299 RCIM_Cubic>]}
       }})
     Scale Factor=&MakeVector4_FromFloat(Float=&Multiply_Float(A=8, B=&FNTimeOfDayScalar(Night Scale=1, Morning Scale=0.5, Evening Scale=0.5)))
    )) {isSetParameters=true}
   &ParticleState()
   &SolveForcesAndVelocity(Clamp Velocity=false, Limit Acceleration=false)
   &SkeletalMeshLocation(
    Bone / Socket Index=$Particles.SpawnGroup
    Bone Sampling Mode=::NewEnumerator4
    Inherit Velocity=false
    Interpolate Mesh Sampling=false
    MaskSpawnGroup=true
    Sampled Orientation Coordinate Space=::World
    Sampled Position Offset Coordinate Space:ENiagaraMeshTransforms=::NewEnumerator2
    Skeletal Mesh=DataInterface() {propertyValues {
      DefaultMesh=null
      ExcludeBoneName=root
      OptionalComponentClass=null
      PreviewMesh=&M_Med_Soldier_04_Preview
      SkinningMode=SkinOnTheFly
      SoftSourceActor=null
      SourceMode=Default
      UvSetIndex=0
      WholeMeshLOD=-1
      bExcludeBone=false
      bReadDeformedGeometry=true
      bRequireCurrentFrameData=true
      ComponentTags []
      FilteredBones [hand_l,hand_r,foot_l,foot_r]
      FilteredSockets []
      MeshUserParameter {parameter {name=None, typeDefHandle {registeredTypeIndex=1}, varData []}}
      SamplingRegions []
     }}
   ) {module=SkeletalMeshLocation002}
   /SetVariables(Particles.Light_position:Vector3f=&Add_Vector(A=Linked() {linkedVariable {name=Particles.Position, type=NiagaraPosition}}, B=<0 0 -10>)) {isSetParameters=true}
   &ScaleColor(
    Color Value To Scale=$Particles.Color
    Scale Alpha=&Multiply_Float(A=2, B=&FNTimeOfDayScalar(Night Scale=1, Morning Scale=0.5, Evening Scale=0.5))
    ScaleA=true
    ScaleRGB=true
   )
  ]
  rendererClasses [&NiagaraSpriteRendererProperties, &NiagaraLightRendererProperties]
  renderers [
   {rendererClass=&NiagaraSpriteRendererProperties}
   {rendererClass=&NiagaraLightRendererProperties, rendererIndex=1}
  ]
 }
 Drops {
  simTarget=GPUComputeSim
  emitterUpdateScript [
   &EmitterState(Inactive Response=::NewEnumerator1, Life Cycle Mode=::NewEnumerator1, Loop Behavior=::NewEnumerator1, Loop Duration=3600, UseLoopDelay=false)
   &SpawnRate(
    SpawnGroup:NiagaraInt32=&RandomRangeInt(Minimum=0, Maximum=1)
    SpawnRate=50
    Use Spawn Probability=false
   )
  ]
  particleSpawnScript [
   &FN_SetSpawnGroup(Number of Spawn Groups=2)
   &InitializeParticle(
    Color=&MakeCustomLinearColorFromBool(
     Bool=&RandomBool()
     True Color=&MakeCustomLinearColorFromBool(Bool=&RandomBool(), True Color=<rgba 0.8 0.047 0.1108 1>, False Color=<rgba 0.1667 0 1 1>)
     False Color=&MakeCustomLinearColorFromBool(Bool=&RandomBool(), True Color=<rgba 5 1.1051 0 1>, False Color=<rgba 0 1 0.3163 1>)
    )
    Lifetime Max=0.5
    Lifetime Min=0.5
    Lifetime Mode=::NewEnumerator1
    Sprite Size Mode=::NewEnumerator1
    Uniform Sprite Size Min=10
    UsePositionOffset=false
    Write Material Random=true
   ) {module=InitializeParticle001}
   &SkeletalMeshLocation(
    Bone / Socket Index=$Particles.SpawnGroup
    Bone Sampling Mode=::NewEnumerator4
    Inherit Velocity=false
    Interpolate Mesh Sampling=false
    MaskSpawnGroup=false
    Sampled Position Offset Coordinate Space:ENiagaraMeshTransforms=::NewEnumerator3
    Skeletal Mesh=$User.SkeletalMesh
   )
   /SetVariables(Particles.Velocity=&UniformRangedVector(Minimum=<-100 -100 250>, Maximum=<100 100 400>)) {isSetParameters=true}
   /SetVariables(Particles.SpriteSize=&Multiply_Vector2(
     A=<5 10>
     B=&Vector2DFromFloat(Value=&UniformRangedFloat(Minimum=1, Maximum=2))
    )) {isSetParameters=true}
  ]
  particleUpdateScript [
   &ParticleState()
   &GravityForce(Gravity=<0 0 -100>)
   &Drag(Use Linear Drag=true, Use Rotational Drag=false)
   &SolveForcesAndVelocity(Clamp Velocity=false, Limit Acceleration=false)
   &ScaleColor(
    Scale Alpha=&FloatFromCurve(FloatCurve=DataInterface() {propertyValues {Curve {keys [
         <at 0 0 RCIM_Cubic>
         {arriveTangent=1.6041, leaveTangent=1.6041, time=0.05, value=1.0113}
         {interpMode=RCIM_Cubic, leaveTangent=-1.0633, time=0.75, value=1}
         <at 0.99 0 RCIM_Cubic>
        ]}}})
    Scale RGB=<0.6 0.8 1>
    ScaleA=true
    ScaleRGB=true
   )
   &ScaleSpriteSize(Scale Factor=&Vector2DFromFloat(Value=&FloatFromCurve(
      FloatCurve=DataInterface() {propertyValues {Curve {keys [
          {arriveTangent=-0.4919, interpMode=RCIM_Cubic, leaveTangent=-0.4919, tangentMode=RCTM_User, time=0.25, value=2}
          {arriveTangent=-3.3301, interpMode=RCIM_Cubic, leaveTangent=-3.3302, tangentMode=RCTM_User, time=1, value=0}
         ]}}}
      Scale Curve=0.65
     )))
  ]
  rendererClasses [&NiagaraSpriteRendererProperties]
  renderers [{rendererClass=&NiagaraSpriteRendererProperties}]
 }
 Mesh_ribbon {
  emitterUpdateScript [
   &EmitterState(Inactive Response=::NewEnumerator1, Life Cycle Mode=::NewEnumerator1, Loop Duration=1, UseLoopDelay=false)
   &SpawnRate(SpawnRate=5, Use Spawn Probability=false)
   &SpawnBurst_Instantaneous(Spawn Count=1, Use Spawn Probability=false)
  ]
  particleSpawnScript [
   &FN_SetSpawnGroup(Number of Spawn Groups=2)
   &InitializeParticle(
    Color=&MakeCustomLinearColorFromBool(
     Bool=&RandomBool()
     True Color=&MakeCustomLinearColorFromBool(Bool=&RandomBool(), True Color=<rgba 0.8 0 0.0808 1>, False Color=<rgba 0.1667 0 1 1>)
     False Color=&MakeCustomLinearColorFromBool(Bool=&RandomBool(), True Color=<rgba 5 1.1051 0 1>, False Color=<rgba 0 1 0.3163 1>)
    )
    Lifetime Max=1
    Lifetime Min=0.8
    Lifetime Mode=::NewEnumerator1
    Mesh Scale=&RandomRangeVector(Minimum=<1 1 1>, Maximum=<0.7 0.7 1.5>)
    Mesh Scale Mode=::NewEnumerator4
    UsePositionOffset=false
    Write Material Random=true
   ) {module=InitializeParticle001}
   &InitialMeshOrientation(
    Rotation=&RandomRangeVector(Minimum=<0 0 0>, Maximum=<0 0 1>)
    Use Orientation Vector=true
    Use Rotation Vector=true
   )
   &AddVelocity()
  ]
  particleUpdateScript [
   &SkeletalMeshLocation(
    Bone / Socket Index=$Particles.SpawnGroup
    Bone Sampling Mode=::NewEnumerator4
    Inherit Velocity=false
    Interpolate Mesh Sampling=false
    MaskSpawnGroup=false
    Sampled Position Offset Coordinate Space:ENiagaraMeshTransforms=::NewEnumerator3
    Skeletal Mesh=$User.SkeletalMesh
   )
   &ParticleState()
   &SolveForcesAndVelocity(Clamp Velocity=false, Limit Acceleration=false)
   &DynamicMaterialParameters(
    Index 0 Param 1=&FloatFromCurve(FloatCurve=DataInterface() {propertyValues {Curve {keys [<at 0 0>,<at 1 1.2>]}}})
    Index 0 Param 2=&RandomRangeFloat(Minimum=3, Maximum=5)
    Index 0 Param 3=&Multiply_Float(
     A=&RandomRangeFloat(Minimum=0, Maximum=1)
     B=&FloatFromCurve(FloatCurve=DataInterface() {propertyValues {Curve {keys [<at 0 1>,<at 1 1>]}}})
    )
    Module.Param0WriteEnabled=true
    Module.Param1WriteEnabled=true
    Module.Param2WriteEnabled=true
    Module.Param3WriteEnabled=true
   )
  ]
  rendererClasses [&NiagaraMeshRendererProperties]
  renderers [{rendererClass=&NiagaraMeshRendererProperties, rendererIndex=0}]
 }
 Paint_Ribbon01 {
  emitterUpdateScript [
   &EmitterState()
   /Game/Effects/Niagara/Modules/Ribbons/FN_SpawnRibbon_v3(
    Number of Ribbons=2
    Skeleton Mesh=DataInterface() {propertyValues {
      DefaultMesh=null
      ExcludeBoneName=root
      OptionalComponentClass=null
      PreviewMesh=&M_Med_Soldier_04_Preview
      SkinningMode=SkinOnTheFly
      SoftSourceActor=null
      SourceMode=Default
      UvSetIndex=0
      WholeMeshLOD=-1
      bExcludeBone=false
      bReadDeformedGeometry=true
      bRequireCurrentFrameData=true
      ComponentTags []
      FilteredBones [hand_l,hand_r]
      FilteredSockets []
      MeshUserParameter {parameter {name=None, typeDefHandle {registeredTypeIndex=1}, varData []}}
      SamplingRegions []
     }}
   )
  ]
  particleSpawnScript [
   &FN_SetSpawnGroup(Number of Spawn Groups=2)
   /Game/Effects/Niagara/Modules/Ribbons/FN_InitializeRibbon_v5(Enable Skeletal Position=true, Lifetime=1, Ribbon Width=50, SkeletalMesh:NiagaraDataInterfaceSkeletalMesh=$User.SkeletalMesh)
   &AddVelocity(Velocity=&FN_FrontEndVector3Modifier(Gameplay=<0 0 300>, Front End=<0 0 400>)) {module=AddVelocity001}
   /SetVariables(Particles.Color=&LerpLinearColors(
     StartColor=&ColorFromCurve(
      ColorCurve=DataInterface() {propertyValues {
        AlphaCurve {keys [<at 0 1 RCIM_Cubic>]}
        BlueCurve {keys [<at 0 1 RCIM_Cubic>,<at 0.0929 1>,<at 0.5021 1>,<at 0.6282 0.2097>,<at 1 0.1955 RCIM_Cubic>]}
        GreenCurve {keys [<at 0 0.7559 RCIM_Cubic>,<at 0.0929 0.0537>,<at 0.5021 0.7457>,<at 0.6282 0>,<at 1 0.0254 RCIM_Cubic>]}
        RedCurve {keys [<at 0 0 RCIM_Cubic>,<at 0.0929 0>,<at 0.5021 0>,<at 0.6282 0.651>,<at 1 1 RCIM_Cubic>]}
       }}
      CurveIndex=&RandomRangeFloat(Minimum=0, Maximum=1)
     )
     EndColor=&ColorFromCurve(
      ColorCurve=DataInterface() {propertyValues {
        AlphaCurve {keys [<at 0 1 RCIM_Cubic>]}
        BlueCurve {keys [<at 0 1 RCIM_Cubic>,<at 0.0929 1>,<at 0.5021 1>,<at 0.5906 0.2097>,<at 1 1 RCIM_Cubic>]}
        GreenCurve {keys [<at 0 0.7559 RCIM_Cubic>,<at 0.0929 0.0537>,<at 0.5021 0.7457>,<at 0.5906 0>,<at 1 0.0833 RCIM_Cubic>]}
        RedCurve {keys [<at 0 0 RCIM_Cubic>,<at 0.0929 0>,<at 0.5021 0>,<at 0.5906 0.651>,<at 1 0.5317 RCIM_Cubic>]}
       }}
      CurveIndex=&RandomRangeFloat(Minimum=0, Maximum=1)
     )
     LerpFactor=&MakeFloatFromInt(Int=$Particles.SpawnGroup)
    )) {isSetParameters=true, module=SetVariables_#1}
   /SetVariables(Particles.MaterialRandom:NiagaraFloat=&RandomRangeFloat(Minimum=0, Maximum=1)) {isSetParameters=true}
   /SetVariables() {isSetParameters=true}
  ]
  particleUpdateScript [
   &ScaleVelocity(Velocity Scale=&MakeVector(
     X=0
     Y=0
     Z=&Multiply_Float(
      A=0.9
      B=&FloatFromCurve(
       FloatCurve=DataInterface() {propertyValues {Curve {keys [<at 0 0.9>,<at 1 3>]}}}
       CurveIndex=&Divide_Float(A=&VectorLength(Vector=$Engine.Owner.Velocity), B=9000)
      )
     )
    )) {module=ScaleVelocity001}
   &SolveForcesAndVelocity(Clamp Velocity=false, Limit Acceleration=false)
   /Game/Effects/Niagara/Modules/Ribbons/FN_UpdateRibbon_v6(Enable Skeletal Position=true, SkeletalMesh:NiagaraDataInterfaceSkeletalMesh=$User.SkeletalMesh)
   &DynamicMaterialParameters(
    Index 0 Param 1=&MaskFloatBySpawnGroup(Masked Float=0, Passthrough Float=0.5)
    Index 0 Param 2=&FloatFromCurve(FloatCurve=DataInterface() {propertyValues {Curve {keys [
         {arriveTangent=-0.3062, interpMode=RCIM_Cubic, leaveTangent=-0.3062, tangentMode=RCTM_User, time=0, value=0.5}
         {arriveTangent=0.0167, interpMode=RCIM_Cubic, leaveTangent=0.0167, tangentMode=RCTM_Break, time=1, value=0}
        ]}}})
    Index 0 Param 3=1
    Index 0 Param 4=1
    Module.Param0WriteEnabled=true
    Module.Param1WriteEnabled=true
    Module.Param2WriteEnabled=true
    Module.Param3WriteEnabled=true
   )
   &ScaleRibbonWidth(Ribbon Width Scale=&FloatFromCurve(
     FloatCurve=DataInterface() {propertyValues {Curve {keys [
         {arriveTangent=-0.0063, interpMode=RCIM_Cubic, leaveTangent=-0.0063, tangentMode=RCTM_User, time=0, value=1}
         {arriveTangent=2.3571, interpMode=RCIM_Cubic, leaveTangent=2.3571, time=0.4, value=2.025}
         <at 0.7 2.325>
        ]}}}
     Scale Curve=0.6
    ))
   &ScaleColor(
    Scale Alpha=&Multiply_Float(
     A=&FNTimeOfDayScalar(Night Scale=1.5, Morning Scale=2, Evening Scale=2)
     B=&FloatFromCurve(FloatCurve=DataInterface() {propertyValues {Curve {keys [<at 0 0>,<at 0.1 1>,<at 1 1>]}}})
    )
    Scale RGB=&Multiply_Vector(
     A=&MakeVectorFromLinearColorRGB(Color=<rgba 1 1 1 1>)
     B=&VectorFromFloat(Value=1.5)
    )
    ScaleA=true
    ScaleRGB=true
   )
  ]
  rendererClasses [&NiagaraRibbonRendererProperties]
  renderers [{rendererClass=&NiagaraRibbonRendererProperties}]
 }
 Paint_Ribbon02 {
  emitterUpdateScript [
   &EmitterState()
   &SpawnRate(SpawnRate=0, Use Spawn Probability=false)
   &FN_SpawnRibbon_v3(
    Number of Ribbons=2
    Skeleton Mesh=DataInterface() {propertyValues {
      DefaultMesh=null
      ExcludeBoneName=root
      OptionalComponentClass=null
      PreviewMesh=&M_Med_Soldier_04_Preview
      SkinningMode=SkinOnTheFly
      SoftSourceActor=null
      SourceMode=Default
      UvSetIndex=0
      WholeMeshLOD=-1
      bExcludeBone=false
      bReadDeformedGeometry=true
      bRequireCurrentFrameData=true
      ComponentTags []
      FilteredBones [hand_l,hand_r]
      FilteredSockets []
      MeshUserParameter {parameter {name=None, typeDefHandle {registeredTypeIndex=1}, varData []}}
      SamplingRegions []
     }}
   )
  ]
  particleSpawnScript [
   &FN_SetSpawnGroup(Number of Spawn Groups=2)
   &FN_InitializeRibbon_v5(Enable Skeletal Position=true, Lifetime=0.8, Ribbon Width=50, SkeletalMesh:NiagaraDataInterfaceSkeletalMesh=$User.SkeletalMesh)
   &AddVelocity(Velocity=&FN_FrontEndVector3Modifier(Gameplay=<0 0 300>, Front End=<0 0 400>)) {module=AddVelocity001}
   /SetVariables(Particles.Color=&LerpLinearColors(
     StartColor=&ColorFromCurve(
      ColorCurve=DataInterface() {propertyValues {
        AlphaCurve {keys [<at 0 1 RCIM_Cubic>]}
        BlueCurve {keys [<at 0 1 RCIM_Cubic>,<at 0.0929 1>,<at 0.5021 0>,<at 0.7704 0.2097>,<at 1 0.3701 RCIM_Cubic>]}
        GreenCurve {keys [<at 0 0.7559 RCIM_Cubic>,<at 0.0929 0.0537>,<at 0.5021 0.8217>,<at 0.7704 0>,<at 1 0.0135 RCIM_Cubic>]}
        RedCurve {keys [<at 0 0 RCIM_Cubic>,<at 0.0929 0>,<at 0.5021 1>,<at 0.7704 0.651>,<at 1 1 RCIM_Cubic>]}
       }}
      CurveIndex=&RandomRangeFloat(Minimum=0, Maximum=1)
     )
     EndColor=&ColorFromCurve(
      ColorCurve=DataInterface() {propertyValues {
        AlphaCurve {keys [<at 0 1 RCIM_Cubic>]}
        BlueCurve {keys [<at 0 1 RCIM_Cubic>,<at 0.0929 1>,<at 0.5021 0>,<at 0.7704 0.2097>,<at 1 0.3701 RCIM_Cubic>]}
        GreenCurve {keys [<at 0 0.7559 RCIM_Cubic>,<at 0.0929 0.0537>,<at 0.5021 0.6901>,<at 0.7704 0>,<at 1 0.0135 RCIM_Cubic>]}
        RedCurve {keys [<at 0 0 RCIM_Cubic>,<at 0.0929 0>,<at 0.5021 1>,<at 0.7704 0.651>,<at 1 1 RCIM_Cubic>]}
       }}
      CurveIndex=&RandomRangeFloat(Minimum=0, Maximum=1)
     )
     LerpFactor=&MakeFloatFromInt(Int=$Particles.SpawnGroup)
    )) {isSetParameters=true, module=SetVariables_#1}
   /SetVariables(Particles.MaterialRandom:NiagaraFloat=&RandomRangeFloat(Minimum=0, Maximum=1)) {isSetParameters=true}
  ]
  particleUpdateScript [
   &ScaleVelocity(Velocity Scale=&MakeVector(
     X=0
     Y=0
     Z=&Multiply_Float(
      A=0.9
      B=&FloatFromCurve(
       FloatCurve=DataInterface() {propertyValues {Curve {keys [<at 0 0.9>,<at 1 3>]}}}
       CurveIndex=&Divide_Float(A=&VectorLength(Vector=$Engine.Owner.Velocity), B=9000)
      )
     )
    )) {module=ScaleVelocity001}
   &SolveForcesAndVelocity(Clamp Velocity=false, Limit Acceleration=false)
   &FN_UpdateRibbon_v6(Enable Skeletal Position=true, SkeletalMesh:NiagaraDataInterfaceSkeletalMesh=$User.SkeletalMesh)
   &DynamicMaterialParameters(
    Index 0 Param 1=&MaskFloatBySpawnGroup(Masked Float=0.3, Passthrough Float=1)
    Index 0 Param 2=&FloatFromCurve(FloatCurve=DataInterface() {propertyValues {Curve {keys [
         {arriveTangent=-0.3062, interpMode=RCIM_Cubic, leaveTangent=-0.3062, tangentMode=RCTM_User, time=0, value=0.5}
         {arriveTangent=0.0167, interpMode=RCIM_Cubic, leaveTangent=0.0167, tangentMode=RCTM_Break, time=1, value=0}
        ]}}})
    Index 0 Param 3=1
    Index 0 Param 4=1
    Module.Param0WriteEnabled=true
    Module.Param1WriteEnabled=true
    Module.Param2WriteEnabled=true
    Module.Param3WriteEnabled=true
   )
   &ScaleRibbonWidth(Ribbon Width Scale=&FloatFromCurve(
     FloatCurve=DataInterface() {propertyValues {Curve {keys [
         {arriveTangent=-0.0063, interpMode=RCIM_Cubic, leaveTangent=-0.0063, tangentMode=RCTM_User, time=0, value=1}
         {arriveTangent=2.3571, interpMode=RCIM_Cubic, leaveTangent=2.3571, time=0.4, value=2.025}
         <at 0.7 2.325>
        ]}}}
     Scale Curve=0.6
    ))
   &ScaleColor(
    Scale Alpha=&Multiply_Float(
     A=&FNTimeOfDayScalar(Night Scale=1.5, Morning Scale=2, Evening Scale=2)
     B=&FloatFromCurve(FloatCurve=DataInterface() {propertyValues {Curve {keys [<at 0 0>,<at 0.1 1>,<at 1 1>]}}})
    )
    Scale RGB=&Multiply_Vector(
     A=&MakeVectorFromLinearColorRGB(Color=<rgba 1 1 1 1>)
     B=&VectorFromFloat(Value=2)
    )
    ScaleA=true
    ScaleRGB=true
   ) {module=ScaleColor001}
  ]
  rendererClasses [&NiagaraRibbonRendererProperties]
  renderers [{rendererClass=&NiagaraRibbonRendererProperties}]
 }
 Sparkles {
  simTarget=GPUComputeSim
  emitterUpdateScript [
   &EmitterState(Inactive Response=::NewEnumerator1, Life Cycle Mode=::NewEnumerator1, Loop Duration=0.3, UseLoopDelay=false)
   &SpawnBurst_Instantaneous(Spawn Count=8, Use Spawn Probability=false)
  ]
  particleSpawnScript [
   &FN_SetSpawnGroup(Number of Spawn Groups=2) {module=FN_SetSpawnGroup001}
   /Niagara/Modules/Spawn/Initialization/InitializeParticle(
    Color=&MakeCustomLinearColorFromBool(
     Bool=&RandomBool()
     True Color=&MakeCustomLinearColorFromBool(Bool=&RandomBool(), True Color=<rgba 0.8 0 0.119 1>, False Color=<rgba 0.1667 0 1 1>)
     False Color=&MakeCustomLinearColorFromBool(Bool=&RandomBool(), True Color=<rgba 5 1.1051 0 1>, False Color=<rgba 0 1 0.3163 1>)
    )
    Lifetime=&RandomRangeFloat(Minimum=0.4, Maximum=0.7)
    Module.Write Color=true
    Module.Write Lifetime=true
    Module.Write Mass=true
    Module.Write Material Random=true
    Module.Write Position=false
    Module.Write Scale=false
    Module.Write SpriteRotation=true
    Module.Write SpriteSize=true
    Position=&SimulationPosition()
    Sprite Rotation=&RandomRangeFloat(Minimum=0, Maximum=360)
    Sprite Size=&Vector2DFromFloat(Value=&RandomRangeFloat(Minimum=15, Maximum=5))
   )
   &SkeletalMeshLocation(
    Bone / Socket Index=$Particles.SpawnGroup
    Bone Sampling Mode=::NewEnumerator4
    Inherit Velocity=false
    Interpolate Mesh Sampling=false
    MaskSpawnGroup=false
    Sampled Position Offset=<0 10 0>
    Sampled Position Offset Coordinate Space:ENiagaraMeshTransforms=::NewEnumerator2
    Skeletal Mesh=$User.SkeletalMesh
   )
   &SphereLocation(
    MaskSpawnGroup=false
    Override Seed=false
    Sphere Radius=&RandomRangeFloat(Minimum=2, Maximum=10)
    Surface Only=false
    UseNonUniformScale=false
    UseOffset=false
    UseSphereOrientation=false
   )
   &AddVelocityInCone(Cone Angle=60, Cone Axis=<0 0 1>, Module.Use Velocity Falloff On Cone Axis=false, Velocity Strength=400)
  ]
  particleUpdateScript [
   &ParticleState()
   &Drag(Drag=2, Use Linear Drag=true, Use Rotational Drag=false)
   &GravityForce(Gravity=&RandomRangeVector(Minimum=<0 0 -200>, Maximum=<1 1 -300>))
   /Niagara/Modules/Update/Forces/CurlNoiseForce(Noise Strength=200, Pan Noise=false)
   &SolveForcesAndVelocity(Clamp Velocity=false, Limit Acceleration=false)
   &ScaleColor(Scale Alpha=5, ScaleA=true, ScaleRGB=true)
   &ScaleSpriteSize(Scale Factor=&Vector2DFromFloat(Value=&FloatFromCurve(FloatCurve=DataInterface() {propertyValues {Curve {keys [<at 0.0018 0.1957>,<at 0.0527 0.819>,<at 0.1169 0.6555>,<at 0.1748 0.9015>,<at 0.2394 0.582>,<at 0.332 1.0023>,<at 0.4947 0.3707>,<at 0.6727 1>,<at 1 0>]}}})))
  ]
  rendererClasses [&NiagaraSpriteRendererProperties]
  renderers [{rendererClass=&NiagaraSpriteRendererProperties}]
 }
 Sticker_Feet {
  emitterUpdateScript [
   &EmitterState(Inactive Response=::NewEnumerator1, Life Cycle Mode=::NewEnumerator1, Loop Duration=1, UseLoopDelay=false)
   &SpawnRate(SpawnRate=15, Use Spawn Probability=false)
  ]
  particleSpawnScript [
   &FN_SetSpawnGroup(Number of Spawn Groups=2)
   /Niagara/Modules/Spawn/Initialization/V2/InitializeParticle(
    Lifetime Max=0.3
    Lifetime Min=0.3
    Lifetime Mode=::NewEnumerator1
    Sprite Rotation Mode=::NewEnumerator2
    Sprite Size Mode=::NewEnumerator1
    Uniform Sprite Size Max=20
    Uniform Sprite Size Min=8
    UsePositionOffset=false
    Write Material Random=true
   )
   &SkeletalMeshLocation(
    Bone / Socket Index=$Particles.SpawnGroup
    Bone Sampling Mode=::NewEnumerator4
    Inherit Velocity=false
    Interpolate Mesh Sampling=false
    MaskSpawnGroup=false
    Sampled Position Offset Coordinate Space:ENiagaraMeshTransforms=::NewEnumerator2
    Skeletal Mesh=DataInterface() {propertyValues {
      DefaultMesh=null
      ExcludeBoneName=root
      OptionalComponentClass=null
      PreviewMesh=&M_Med_Soldier_04_Preview
      SkinningMode=SkinOnTheFly
      SoftSourceActor=null
      SourceMode=Default
      UvSetIndex=0
      WholeMeshLOD=-1
      bExcludeBone=false
      bReadDeformedGeometry=true
      bRequireCurrentFrameData=true
      ComponentTags []
      FilteredBones [foot_l,foot_r]
      FilteredSockets []
      MeshUserParameter {parameter {name=None, typeDefHandle {registeredTypeIndex=1}, varData []}}
      SamplingRegions []
     }}
   )
   &AddVelocityInCone(Cone Angle=0, Cone Axis=<0 0 1>, Module.Use Velocity Falloff On Cone Axis=false, Velocity Strength=100)
   &SubUVAnimation(End Frame=15, SubUV Animation Mode:ENiagara_SubUVLookupMode=::NewEnumerator2)
   &CameraOffset(Camera Offset Amount=20)
  ]
  particleUpdateScript [
   &ParticleState()
   &SolveForcesAndVelocity(Clamp Velocity=false, Limit Acceleration=false)
   &ScaleSpriteSize(Scale Factor=&Vector2DFromFloat(Value=&FloatFromCurve(FloatCurve=DataInterface() {propertyValues {Curve {keys [<at 0 0>,<at 0.2 1>,<at 0.7 1>,<at 1 0.2>]}}})))
   &SpriteRotationRate(Delta Time:NiagaraFloat=$Particles.Lifetime, Rotation Rate=&RandomRangeFloat(Minimum=-3, Maximum=3))
  ]
  rendererClasses [&NiagaraSpriteRendererProperties]
  renderers [{rendererClass=&NiagaraSpriteRendererProperties}]
 }
 Sticker_Hands {
  emitterUpdateScript [
   &EmitterState(Inactive Response=::NewEnumerator1, Life Cycle Mode=::NewEnumerator1, Loop Duration=1, UseLoopDelay=false)
   &SpawnRate(SpawnRate=30, Use Spawn Probability=false)
  ]
  particleSpawnScript [
   &FN_SetSpawnGroup(Number of Spawn Groups=2)
   &InitializeParticle(
    Lifetime Max=1
    Lifetime Min=0.7
    Lifetime Mode=::NewEnumerator1
    Sprite Rotation Mode=::NewEnumerator2
    Sprite Size Mode=::NewEnumerator1
    Uniform Sprite Size Max=25
    Uniform Sprite Size Min=10
    UsePositionOffset=false
    Write Material Random=true
   )
   &SkeletalMeshLocation(
    Bone / Socket Index=$Particles.SpawnGroup
    Bone Sampling Mode=::NewEnumerator4
    Inherit Velocity=false
    Interpolate Mesh Sampling=false
    MaskSpawnGroup=false
    Sampled Position Offset=<0 10 0>
    Sampled Position Offset Coordinate Space:ENiagaraMeshTransforms=::NewEnumerator2
    Skeletal Mesh=$User.SkeletalMesh
   )
   &AddVelocityInCone(
    Cone Angle=40
    Cone Axis=<0 0 1>
    Module.Use Velocity Falloff On Cone Axis=false
    Velocity Strength=&RandomRangeFloat(Minimum=300, Maximum=400)
   )
   &SubUVAnimation(End Frame=15, SubUV Animation Mode:ENiagara_SubUVLookupMode=::NewEnumerator2)
  ]
  particleUpdateScript [
   &ParticleState()
   &ScaleVelocity(Velocity Scale=&MakeVector(
     X=0.6
     Y=0.6
     Z=&Multiply_Float(
      A=0.6
      B=&FloatFromCurve(
       FloatCurve=DataInterface() {propertyValues {Curve {keys [<at 0 0.9>,<at 1 3>]}}}
       CurveIndex=&Divide_Float(A=&VectorLength(Vector=$Engine.Owner.Velocity), B=9000)
      )
     )
    )) {module=ScaleVelocity001}
   &GravityForce(Gravity=&RandomRangeVector(Minimum=<0 0 -200>, Maximum=<1 1 -300>)) {module=GravityForce001}
   &Drag(Drag=3, Use Linear Drag=true, Use Rotational Drag=false)
   &SolveForcesAndVelocity(Clamp Velocity=false, Limit Acceleration=false)
   &ScaleSpriteSize(Scale Factor=&Vector2DFromFloat(Value=&FloatFromCurve(FloatCurve=DataInterface() {propertyValues {Curve {keys [<at 0 0>,<at 0.2 1>,<at 0.6 1>,<at 1 0.2>]}}})))
   &SpriteRotationRate(Delta Time:NiagaraFloat=$Particles.Lifetime, Rotation Rate=&RandomRangeFloat(Minimum=-3, Maximum=3))
  ]
  rendererClasses [&NiagaraSpriteRendererProperties]
  renderers [{rendererClass=&NiagaraSpriteRendererProperties}]
 }
}
summary {
 systemName=NS_Contrail_ColorTrip
 emitters [
  {emitterName=Paint_Ribbon01, rendererClasses [&NiagaraRibbonRendererProperties]}
  {emitterName=Paint_Ribbon02, rendererClasses [&NiagaraRibbonRendererProperties]}
  {emitterName=Sticker_Hands, rendererClasses [&NiagaraSpriteRendererProperties]}
  {emitterName=Drops, simTarget=GPUComputeSim, rendererClasses [&NiagaraSpriteRendererProperties]}
  {emitterName=Sticker_Feet, rendererClasses [&NiagaraSpriteRendererProperties]}
  {emitterName=Sparkles, simTarget=GPUComputeSim, rendererClasses [&NiagaraSpriteRendererProperties]}
  {emitterName=AttachedGlow, rendererClasses [&NiagaraSpriteRendererProperties, &NiagaraLightRendererProperties]}
  {emitterName=Mesh_ribbon, rendererClasses [&NiagaraMeshRendererProperties]}
 ]
 userVariables [
  User.bIsFrontEnd {type=NiagaraBool}
  User.bIsFrontEndPreview {type=NiagaraBool}
  User.PlayerSpeed {type=NiagaraFloat, defaultValue=0, description=""}
  User.RibbonIDOffset {type=NiagaraInt32}
  User.SkeletalMesh {
   type=NiagaraDataInterfaceSkeletalMesh
   defaultValue {
    struct=NiagaraExt_VariableValue_DataInterface
    value {dataInterface=&self:NiagaraDataInterfaceSkeletalMesh_5}
   }
  }
 ]
}
